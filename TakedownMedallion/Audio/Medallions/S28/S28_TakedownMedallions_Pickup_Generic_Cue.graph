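# Season 28 sound cue for the generic Takedown Medallion pickup sound, in the loot collect sound class.
# A single wave player feeds the default medallion pickup wave through a Modulator node left at default settings.
# A concurrency override limits the cue to one instance at a time.
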SoundCue /TakedownMedallion/Audio/Medallions/S28/S28_TakedownMedallions_Pickup_Generic_Cue
root=Modulator_0
nodes [
 {id=Modulator_0, children [WavePlayer_5]}
 {id=WavePlayer_5, properties {soundWaveAssetPtr=&/Default_TakedownMedallions_Pickup_SFX}}
]
properties {
 bOverrideConcurrency=true
 maxDistance=2097152
 soundClassObject=&Fort_Loot_Collect
 volumeMultiplier=1
 concurrencyOverrides {maxCount=1}
}
