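# Sound cue for picking up a Season 28 Takedown Medallion, in its PG variant.
# A mixer layers the default medallion pickup SFX with an S28 PG "Mx" wave, which starts after a short fixed delay and sits slightly lower in the mix.
# It plays through the loot-collect sound class, and a concurrency override limits it to one instance at a time.
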SoundCue /TakedownMedallion/Audio/Medallions/S28/S28_TakedownMedallions_Pickup_PG_Cue
root=Mixer_2
nodes [
 {id=Delay_0, children [WavePlayer_1], properties {delayMax=0.15, delayMin=0.15}}
 {id=Mixer_2, children [WavePlayer_5,Delay_0], properties {inputVolume [1,0.75]}}
 {id=WavePlayer_1, properties {soundWaveAssetPtr=&/S28_TakedownMedallions_Pickup_Mx_PG}}
 {id=WavePlayer_5, properties {soundWaveAssetPtr=&/Default_TakedownMedallions_Pickup_SFX}}
]
properties {
 bOverrideConcurrency=true
 maxDistance=2097152
 soundClassObject=&Fort_Loot_Collect
 volumeMultiplier=0.7692
 concurrencyOverrides {maxCount=1}
}
